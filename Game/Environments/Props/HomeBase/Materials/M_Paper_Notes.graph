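# M_Paper_Notes is a surface material for paper note props in the HomeBase environment.
# It is a two-sided Default Lit material flagged for use with Nanite.
# Its only node samples T_Paper_Notes_D, a texture stored in the SurvivorRadio prototype mission folder, with no output wiring shown.
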Material /Game/Environments/Props/HomeBase/Materials/M_Paper_Notes
graph {
 properties {bUsedWithNanite=true, pixelDepthOffsetMode=PDOM_Legacy, twoSided=true, usedShadingModels=Default Lit}
 nodes [TextureSample() {texture=/Game/Missions/Prototypes/Survivor_Stories/SurvivorRadio/Textures/T_Paper_Notes_D}]
}
